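# Translucent unlit, two-sided effect material for the Hightower Date skydiving cosmetic, with default ribbon textures and greens, usable on Niagara ribbons, sprites and meshes.
# A panning color texture blends two colors, tinted by particle color, for emissive.
# Opacity comes from a panning base texture warped by macro and micro noise, multiplied by masks and edge, depth and camera fades, minus noise-driven scrolling erosion.
# Dynamic parameters drive offsets, erosion and optional SplineThicken vertex offset.
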Material /BRCosmetics/Athena/Effects/Particles/Skydiving/NS_Hightower_Date/M_Hightower_Date
graph {
 properties {
  bUsedWithBeamTrails=true
  bUsedWithMeshDeformer=true
  bUsedWithNiagaraMeshParticles=true
  bUsedWithNiagaraRibbons=true
  bUsedWithNiagaraSprites=true
  bUsedWithParticleSprites=true
  bUsedWithSkeletalMesh=true
  bUsedWithSplineMeshes=true
  blendMode=BLEND_Translucent
  pixelDepthOffsetMode=PDOM_Legacy
  shadingModel=MSM_Unlit
  twoSided=true
  usedShadingModels=Unlit
 }
 nodes [
  n1 = ParticleColor()
  n2 = AppendVector(
   A=ScalarParameter() {defaultValue=1.5, group=Color_Texture, parameterName=Color_Texture_Tile_U, sortPriority=3}
   B=ScalarParameter() {defaultValue=1, group=Color_Texture, parameterName=Color_Texture_Tile_V, sortPriority=4}
  )
  n3 = Add(
   A=LinearInterpolate(
    A=TextureCoordinate()
    B=TextureCoordinate() {coordinateIndex=1}
    Alpha=ScalarParameter() {group=BaseTexture, parameterName=Use_UV_1, sortPriority=-1}
   )
   B=AppendVector(
    A=ScalarParameter() {group=Color_Texture, parameterName=Color_Texture_Offset_X, sortPriority=9}
    B=ScalarParameter() {group=Color_Texture, parameterName=Color_Texture_Offset_Y, sortPriority=10}
   )
  )
  n4 = ScalarParameter() {defaultValue=-0.5, group=Color_Texture, parameterName=Color_Texture_PanSpeed_X, sortPriority=1}
  n5 = ScalarParameter() {group=Color_Texture, parameterName=Color_Texture_PanSpeed_Y, sortPriority=2}
  n6 = TextureSampleParameter2D(UVs=Panner(Coordinate=Multiply(A=@n3, B=@n2), Speed=AppendVector(A=@n4, B=@n5)) {speedX=1}) {
   group=Color_Texture
   parameterName=Color_Texture
   samplerSource=SSM_Wrap_WorldGroupSettings
   texture=&/T_Hightower_Ribbon_Color
  }
  n7 = Power(Base=@n6.RGB, Exp=ScalarParameter() {defaultValue=1, group=Color_Texture, parameterName=Color_Texture_Power, sortPriority=7})
  n8 = ScalarParameter() {defaultValue=2, group=Color_Texture, parameterName=Color_Texture_Multiply, sortPriority=8}
  n9 = LinearInterpolate(
   A=VectorParameter() {group=Color_Texture, parameterName=Color_A, sortPriority=5, defaultValue=<rgba 0.8078 1 0.1412 1>}.RGB
   B=VectorParameter() {group=Color_Texture, parameterName=Color_B, sortPriority=6, defaultValue=<rgba 0.2831 1 0.0595 1>}.RGB
   Alpha=Clamp(None=Multiply(A=@n7, B=@n8))
  )
  n10 = LinearInterpolate(
   A=@n1.RGB
   B=LinearInterpolate(A=@n9, B=Multiply(A=@n9, B=@n1.RGB), Alpha=ScalarParameter() {defaultValue=1, parameterName=Use_ParticleColor_ForColor, sortPriority=-1})
   Alpha=ScalarParameter() {defaultValue=1, group=Color_Texture, parameterName=Use_Color_Texture, sortPriority=-1}
  )
  MaterialFunctionCall(Emissive=@n10) {materialFunction=&MF_FX_EmissiveAdjustments}
  Reroute(None=@n10) {desc=Emissive}
  n11 = DynamicParameter() {parameterIndex=1, defaultValue=<rgba 0 0 0 0>, paramNames [Spline_Width_Base, Spline_Width_Tip, DONTUSE, Erode]}
  n12 = ScalarParameter() {group=Spline_WPO, parameterName=Use_DynamicParameter_SplineThicken}
  n13 = Multiply(
   A=ComponentMask(None=TextureCoordinate()) {g=true}
   B=ScalarParameter() {defaultValue=1, group=Spline_WPO, parameterName=WPO_Thicken_Multiply, sortPriority=2}
  )
  n14 = MaterialFunctionCall(
   WidthBase=LinearInterpolate(A=ScalarParameter() {group=Spline_WPO, parameterName=Spline_Width_Base, sortPriority=1}, B=@n11.Spline_Width_Base, Alpha=@n12)
   WidthTip=LinearInterpolate(A=ScalarParameter() {group=Spline_WPO, parameterName=Spline_Width_Tip, sortPriority=2}, B=@n11.Spline_Width_Tip, Alpha=@n12)
   UVs for Thickness=Power(Base=Clamp(None=@n13), Exp=ScalarParameter() {defaultValue=1, group=Spline_WPO, parameterName=WPO_Thicken_Power, sortPriority=3})
  ) {materialFunction=&SplineThicken}
  n15 = DotProduct(A=CameraVectorWS(), B=@n14.Normal)
  n16 = ScalarParameter() {defaultValue=1, parameterName=Camera_Fresnel_Exponent}
  n17 = MaterialFunctionCall(Float2=LinearInterpolate(
    A=TextureCoordinate()
    B=TextureCoordinate() {coordinateIndex=1}
    Alpha=ScalarParameter() {group=BaseTexture, parameterName=Use_UV_1, sortPriority=-1}
   )) {materialFunction=&BreakOutFloat2Components}
  n18 = Multiply(A=@n17.R, B=ScalarParameter() {defaultValue=1, group=MultipliedTexture_B, parameterName=MultipliedTexture_B_Tile_U, sortPriority=2})
  n19 = Multiply(A=@n17.G, B=ScalarParameter() {defaultValue=1, group=MultipliedTexture_B, parameterName=MultipliedTexture_B_Tile_V, sortPriority=3})
  n20 = ScalarParameter() {group=MultipliedTexture_B, parameterName=MultipliedTexture_B_PanSpeed_X, sortPriority=1}
  n21 = ScalarParameter() {group=MultipliedTexture_B, parameterName=MultipliedTexture_B_PanSpeed_Y, sortPriority=2}
  n22 = TextureSampleParameter2D(UVs=Panner(Coordinate=AppendVector(A=@n18, B=@n19), Speed=AppendVector(A=@n20, B=@n21))) {
   group=MultipliedTexture_B
   parameterName=MultipliedTexture_B
   samplerSource=SSM_Wrap_WorldGroupSettings
   sortPriority=1
   texture=&/Hightower_Ribbon_LeadingEdge_Mask
  }
  n23 = AppendVector(
   A=ScalarParameter() {defaultValue=-0.5, group=BaseTexture, parameterName=BaseTexture_PanSpeed_X, sortPriority=1}
   B=ScalarParameter() {group=BaseTexture, parameterName=BaseTexture_PanSpeed_Y, sortPriority=2}
  )
  n24 = DynamicParameter() {parameterIndex=2, defaultValue=<rgba 0 0 1 1>, paramNames [PanSpeed_Mult, MacroDistortion_Offset, "", ""]}
  n25 = Multiply(A=@n23, B=@n24.PanSpeed_Mult)
  n26 = ScalarParameter() {group=NitroFlowOptions, parameterName=NitroFlow_UseDynamicPanSpeed, sortPriority=0}
  n27 = AppendVector(
   A=ScalarParameter() {defaultValue=1, group=BaseTexture, parameterName=Base_Texture_Tile_U, sortPriority=3}
   B=ScalarParameter() {defaultValue=1, group=BaseTexture, parameterName=Base_Texture_Tile_V, sortPriority=4}
  )
  n28 = DynamicParameter() {defaultValue=<rgba 0 0 1 1>, paramNames [Offset U, Offset V, Tiling Multiplier U, Param4]}
  n29 = MaterialFunctionCall(Float2=LinearInterpolate(
    A=TextureCoordinate()
    B=TextureCoordinate() {coordinateIndex=1}
    Alpha=ScalarParameter() {group=BaseTexture, parameterName=Use_UV_1, sortPriority=-1}
   )) {materialFunction=&BreakOutFloat2Components}
  n30 = Multiply(A=@n29.R, B=@n28.Tiling Multiplier U)
  n31 = Multiply(A=@n29.G)
  n32 = MaterialFunctionCall(Float2=LinearInterpolate(
    A=TextureCoordinate()
    B=TextureCoordinate() {coordinateIndex=1}
    Alpha=ScalarParameter() {group=BaseTexture, parameterName=Use_UV_1, sortPriority=-1}
   )) {materialFunction=&BreakOutFloat2Components}
  n33 = Multiply(A=@n32.R, B=ScalarParameter() {defaultValue=0.4, group=Macro_Distortion, parameterName=Distortion_Macro_Tile_X, sortPriority=3})
  n34 = DynamicParameter() {parameterIndex=2, defaultValue=<rgba 0 0 1 1>, paramNames [PanSpeed_Mult, MacroDistortion_Offset, "", ""]}
  n35 = LinearInterpolate(
   A=@n33
   B=Add(A=@n33, B=@n34.MacroDistortion_Offset)
   Alpha=ScalarParameter() {group=NitroFlowOptions, parameterName=NitroFlow_UseDynamicOffset, sortPriority=1}
  )
  n36 = Multiply(A=@n32.G, B=ScalarParameter() {defaultValue=0.2, group=Macro_Distortion, parameterName=Distortion_Macro_Tile_Y, sortPriority=4})
  n37 = ScalarParameter() {defaultValue=-0.45, group=Macro_Distortion, parameterName=Distortion_Macro_Panspeed_X, sortPriority=1}
  n38 = ScalarParameter() {defaultValue=0.01, group=Macro_Distortion, parameterName=Distortion_Macro_Panspeed_Y, sortPriority=2}
  n39 = TextureSampleParameter2D(UVs=Panner(Coordinate=AppendVector(A=@n35, B=@n36), Speed=AppendVector(A=@n37, B=@n38)) {speedY=0.2}) {
   group=Macro_Distortion
   parameterName=Macro_Distortion_Texture
   samplerSource=SSM_Wrap_WorldGroupSettings
   texture=&AtmosphericCloudNoise01
  }
  n40 = Constant()
  n41 = Add(A=Multiply(A=AppendVector(A=@n39.R, B=@n40)) {constB=2}) {constB=-1}
  n42 = Constant()
  n43 = Add(A=Multiply(A=AppendVector(A=@n42, B=@n39.R)) {constB=2}) {constB=-1}
  n44 = ScalarParameter() {defaultValue=1, group=Macro_Distortion, parameterName=MultAdd_XorY, sortPriority=7}
  n45 = ScalarParameter() {group=Macro_Distortion, parameterName=Macro_Distortion_Offset_X, sortPriority=6}
  n46 = ScalarParameter() {group=Macro_Distortion, parameterName=Macro_Distortion_Offset_Y, sortPriority=7}
  n47 = Multiply(
   A=Add(A=LinearInterpolate(A=@n41, B=@n43, Alpha=@n44), B=AppendVector(A=@n45, B=@n46)) {constB=0}
   B=ScalarParameter() {defaultValue=0.0665, group=Macro_Distortion, parameterName=Macro_Distortion_Intensity, sortPriority=5}
  ) {constB=0.05}
  n48 = ScalarParameter() {defaultValue=1, group=Macro_Distortion, parameterName=Use_Macro_Distortion, sortPriority=-1}
  n49 = MaterialFunctionCall(Float2=LinearInterpolate(
    A=TextureCoordinate()
    B=TextureCoordinate() {coordinateIndex=1}
    Alpha=ScalarParameter() {group=BaseTexture, parameterName=Use_UV_1, sortPriority=-1}
   )) {materialFunction=&BreakOutFloat2Components}
  n50 = Multiply(A=@n49.R, B=ScalarParameter() {defaultValue=3, group=Micro_Distortion, parameterName=Distortion_Micro_Tile_X, sortPriority=3})
  n51 = Multiply(A=@n49.G, B=ScalarParameter() {defaultValue=0.5, group=Micro_Distortion, parameterName=Distortion_Micro_Tile_Y, sortPriority=4})
  n52 = ScalarParameter() {defaultValue=-1.7, group=Micro_Distortion, parameterName=Distortion_Micro_Panspeed_X, sortPriority=1}
  n53 = ScalarParameter() {defaultValue=0.05, group=Micro_Distortion, parameterName=Distortion_Micro_Panspeed_Y, sortPriority=2}
  n54 = TextureSampleParameter2D(UVs=Panner(Coordinate=AppendVector(A=@n50, B=@n51), Speed=AppendVector(A=@n52, B=@n53)) {speedY=0.2}) {
   group=Micro_Distortion
   parameterName=Micro_Distortion_Texture
   samplerSource=SSM_Wrap_WorldGroupSettings
   texture=/Game/Global/Textures/Default/Noise/T_Dudebro_Noise_01
  }
  n55 = AppendVector(A=@n54.R, B=@n54.R)
  n56 = Multiply(
   A=Add(A=Multiply(A=@n55) {constB=2}) {constB=-1}
   B=ScalarParameter() {defaultValue=0.0125, group=Micro_Distortion, parameterName=Micro_Distortion_Intensity, sortPriority=5}
  )
  n57 = ScalarParameter() {defaultValue=1, group=Micro_Distortion, parameterName=Use_Micro_Distortion, sortPriority=-1}
  n58 = Add(
   A=Add(A=AppendVector(A=@n30, B=@n31), B=AppendVector(A=@n28.Offset U, B=@n28.Offset U))
   B=Add(A=LinearInterpolate(B=@n47, Alpha=@n48), B=LinearInterpolate(B=@n56, Alpha=@n57))
  )
  n59 = MaterialFunctionCall(Float2=Panner(Coordinate=Multiply(A=@n58, B=@n27), Speed=LinearInterpolate(A=@n23, B=@n25, Alpha=@n26)) {speedX=-0.5}) {materialFunction=&BreakOutFloat2Components}
  n60 = Clamp(None=@n59.R)
  n61 = ScalarParameter() {group=BaseTexture, parameterName=BaseTexture_ClampTiling_X, sortPriority=3}
  n62 = Clamp(None=@n59.G)
  n63 = ScalarParameter() {defaultValue=1, group=BaseTexture, parameterName=BaseTexture_ClampTiling_Y, sortPriority=4}
  n64 = TextureSampleParameter2D(UVs=AppendVector(A=LinearInterpolate(A=@n59.R, B=@n60, Alpha=@n61), B=LinearInterpolate(A=@n59.G, B=@n62, Alpha=@n63))) {
   group=BaseTexture
   parameterName=Base Texture
   samplerSource=SSM_Wrap_WorldGroupSettings
   texture=&/T_Hightower_Ribbon_Macro
  }
  n65 = MaterialFunctionCall(Float2=LinearInterpolate(
    A=TextureCoordinate()
    B=TextureCoordinate() {coordinateIndex=1}
    Alpha=ScalarParameter() {group=BaseTexture, parameterName=Use_UV_1, sortPriority=-1}
   )) {materialFunction=&BreakOutFloat2Components}
  n66 = Multiply(A=@n65.R, B=ScalarParameter() {defaultValue=1, group=MultipliedTexture_A, parameterName=MultipliedTexture_A_Tile_U, sortPriority=2})
  n67 = Multiply(A=@n65.G, B=ScalarParameter() {defaultValue=1, group=MultipliedTexture_A, parameterName=MultipliedTexture_A_Tile_V, sortPriority=3})
  n68 = ScalarParameter() {defaultValue=-0.45, group=MultipliedTexture_A, parameterName=MultipliedTexture_A_PanSpeed_X, sortPriority=1}
  n69 = ScalarParameter() {defaultValue=0.1, group=MultipliedTexture_A, parameterName=MultipliedTexture_A_PanSpeed_Y, sortPriority=2}
  n70 = TextureSampleParameter2D(UVs=Panner(Coordinate=AppendVector(A=@n66, B=@n67), Speed=AppendVector(A=@n68, B=@n69))) {
   group=MultipliedTexture_A
   parameterName=MultipliedTexture_A
   samplerSource=SSM_Wrap_WorldGroupSettings
   sortPriority=1
   texture=&/T_Hightower_Ribbon_Dots
  }
  n71 = ScalarParameter() {defaultValue=1, group=MultipliedTexture_A, parameterName=MultipliedTexture_A_Multiply, sortPriority=4}
  n72 = Clamp(None=Power(
    Base=Multiply(A=@n70.RGB, B=@n71)
    Exp=ScalarParameter() {defaultValue=1, group=MultipliedTexture_A, parameterName=MultipliedTexture_A_Exponent, sortPriority=5}
   ))
  n73 = LinearInterpolate(
   A=@n64.RGB
   B=Multiply(A=MaterialFunctionCall(XYZ=@n72) {materialFunction=&Swizzle}.YXZ, B=@n64.RGB)
   Alpha=ScalarParameter() {defaultValue=1, group=MultipliedTexture_A, parameterName=Use_Multiplied_Texture_A, sortPriority=-1}
  )
  n74 = LinearInterpolate(A=@n73, B=Multiply(A=@n1.A, B=@n73), Alpha=ScalarParameter() {defaultValue=1, parameterName=Use_ParticleColor_ForAlpha, sortPriority=-1})
  n75 = LinearInterpolate(
   A=@n74
   B=Power(Base=@n74, Exp=ScalarParameter() {group=NitroFlowOptions, parameterName=NitroFlow_BasePower, sortPriority=1})
   Alpha=ScalarParameter() {group=NitroFlowOptions, parameterName=NitroFlow_UseBasePower, sortPriority=0}
  )
  n76 = LinearInterpolate(
   A=@n75
   B=Multiply(A=MaterialFunctionCall(XYZ=@n22.RGB) {materialFunction=&Swizzle}.YXZ, B=@n75)
   Alpha=ScalarParameter() {group=MultipliedTexture_B, parameterName=Use_Multiplied_Texture_B, sortPriority=-1}
  )
  n77 = DepthFade(FadeDistance=ScalarParameter() {defaultValue=250, group=Fade, parameterName=DepthFade_Distance})
  n78 = Multiply(
   A=ComponentMask(None=TextureCoordinate()) {r=true}
   B=ScalarParameter() {defaultValue=20, group=Fade, parameterName=LeadingEdge_Fade_UVMultiply, sortPriority=3}
  )
  n79 = Add(
   A=Power(Base=Clamp(None=@n78), Exp=ScalarParameter() {defaultValue=2.2, group=Fade, parameterName=LeadingEdge_Fade_Power, sortPriority=2}) {constExponent=2.2}
   B=OneMinus(None=ScalarParameter() {defaultValue=1, group=Fade, parameterName=Use_LeadingEdgeFade, sortPriority=1})
  )
  n80 = OneMinus(None=ComponentMask(None=TextureCoordinate()) {r=true})
  n81 = ScalarParameter() {defaultValue=20, group=Fade, parameterName=TrailingEdge_Fade_UVMultiply, sortPriority=3}
  n82 = Power(
   Base=Clamp(None=Multiply(A=@n80, B=@n81))
   Exp=ScalarParameter() {defaultValue=2.2, group=Fade, parameterName=TrailingEdge_Fade_Power, sortPriority=2}
  ) {constExponent=2.2}
  n83 = OneMinus(None=ScalarParameter() {group=Fade, parameterName=Use_TrailingEdgeFade, sortPriority=1})
  n84 = Multiply(A=Multiply(A=Multiply(A=@n76, B=@n77), B=Clamp(None=@n79)), B=Clamp(None=Add(A=@n82, B=@n83)))
  n85 = LinearInterpolate(A=@n84, B=Multiply(A=@n84, B=Power(Base=@n15, Exp=@n16)), Alpha=ScalarParameter() {parameterName=Use_CameraFresnel})
  n86 = Power(
   Base=MaterialFunctionCall(
    Fade Length=ScalarParameter() {group=Fade, parameterName=CameraFade_Length, sortPriority=4}
    Fade Offset=ScalarParameter() {group=Fade, parameterName=CameraFade_Offset, sortPriority=5}
   ) {materialFunction=&CameraDepthFade}.Result
   Exp=ScalarParameter() {defaultValue=1, group=Fade, parameterName=CameraFade_Power, sortPriority=6}
  )
  n87 = LinearInterpolate(
   A=TextureCoordinate()
   B=TextureCoordinate() {coordinateIndex=1}
   Alpha=ScalarParameter() {group=Erosion, parameterName=ErosionNoise_UseUV-1, sortPriority=2}
  )
  n88 = AppendVector(
   A=ScalarParameter() {defaultValue=1, group=Erosion, parameterName=ErosionNoise_Tile_X, sortPriority=3}
   B=ScalarParameter() {defaultValue=1, group=Erosion, parameterName=ErosionNoise_Tile_Y, sortPriority=3}
  )
  n89 = ScalarParameter() {group=Erosion, parameterName=ErosionNoise_Panspeed_X, sortPriority=3}
  n90 = ScalarParameter() {group=Erosion, parameterName=ErosionNoise_Panspeed_Y, sortPriority=3}
  n91 = TextureSampleParameter2D(UVs=Panner(Coordinate=Multiply(A=@n87, B=@n88), Speed=AppendVector(A=@n89, B=@n90)) {speedY=0.2}) {
   group=Erosion
   parameterName=ErosionNoise
   samplerType=SAMPLERTYPE_Masks
   sortPriority=1
   texture=/Game/Global/Textures/Default/Noise/T_CellularNoise
  }
  n92 = Power(Base=@n91.RGB, Exp=ScalarParameter() {defaultValue=1, group=Erosion, parameterName=ErosionNoise_Texture-Power, sortPriority=3})
  n93 = ScalarParameter() {defaultValue=1, group=Erosion, parameterName=ErosionNoise_Texture-Hardness, sortPriority=3}
  n94 = ScalarParameter() {group=Erosion, parameterName=Erosion, sliderMax=1, sortPriority=0}
  n95 = LinearInterpolate(A=@n11.Erode, B=ParticleColor().A, Alpha=ScalarParameter() {group=Erosion, parameterName=Use_ParticleAlpha4Erosion, sortPriority=-1})
  n96 = Add(A=Multiply(A=Add(A=@n95, B=@n94)) {constA=2, constB=2}) {constB=-1}
  n97 = VectorParameter() {group=Erosion, parameterName=ScrollingErosion-Direction, sortPriority=33, defaultValue=<rgba 1 0 0 1>}
  n98 = Multiply(A=TextureCoordinate(), B=AppendVector(A=@n97.R, B=@n97.R))
  n99 = Power(Base=Desaturation(None=Add(A=@n96, B=@n98)), Exp=ScalarParameter() {defaultValue=1, group=Erosion, parameterName=ScrollingErosion-Exponent})
  n100 = ScalarParameter() {defaultValue=1, group=Erosion, parameterName=ScrollingErosion-Hardness}
  n101 = Multiply(A=Clamp(None=Multiply(A=@n99, B=@n100))) {constA=2, constB=2}
  n102 = Add(
   A=Clamp(None=Multiply(A=@n92, B=@n93))
   B=LinearInterpolate(
    A=@n96
    B=Add(A=@n101) {constB=-1}
    Alpha=ScalarParameter() {defaultValue=1, group=Erosion, parameterName=Use-ScrollingErosion, sortPriority=31}
   )
  )
  Reroute(None=Subtract(A=Multiply(A=@n85, B=@n86), B=Clamp(None=@n102))) {desc=Opacity}
  Reroute(None=StaticSwitch(
    True=@n14.WorldPositionOffset
    False=Constant()
    Value=StaticBoolParameter() {group=Spline_WPO, parameterName=Use_WPO_SplineThicken, sortPriority=-1}
   )) {desc=WPO}
 ]
}
