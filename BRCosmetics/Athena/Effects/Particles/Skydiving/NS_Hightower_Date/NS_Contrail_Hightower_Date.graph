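# Niagara skydiving contrail for the Hightower_Date cosmetic, trailing green energy and dark smoke from two sockets on the character's skeletal mesh, apparently the hands.
# Spawn groups split particles between sockets for hand glows with lights, black smoke, dark dots and flickering lightning sprites.
# Two ribbons rise from the sockets with pulsing green color, using dive attributes to drive material parameters.
# Randomly timed bursts of dots orbit upward with dive-scaled velocity.
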NiagaraSystem /BRCosmetics/Athena/Effects/Particles/Skydiving/NS_Hightower_Date/NS_Contrail_Hightower_Date
dependencies {
 usedDataInterfaces [&NiagaraDataInterfaceCurve, &NiagaraDataInterfaceVectorCurve, &NiagaraDataInterfaceColorCurve]
 usedDynamicInputs []
 usedModules [&SystemState]
}
emitters {
 Hand_Dots {
  emitterSpawnScript [&FN_PreserveSkydiveUserVariables() {module=PreserveUserVariables}]
  emitterUpdateScript [
   &EmitterState()
   &SpawnRate(SpawnRate=40, Use Spawn Probability=false)
  ]
  particleSpawnScript [
   &FN_SetSpawnGroup_v2(Number of Spawn Groups=2)
   /Niagara/Modules/Spawn/Initialization/InitializeParticle(
    Color=<rgba 0 0 0 1>
    Lifetime=1.5
    Mesh Scale=<5 5 5>
    Module.Write Color=true
    Module.Write Lifetime=true
    Module.Write Mass=true
    Module.Write Material Random=true
    Module.Write Position=false
    Module.Write Scale=false
    Module.Write SpriteRotation=false
    Module.Write SpriteSize=true
    Position=&SimulationPosition()
    Sprite Size=&Vector2DFromFloat(Value=&UniformRangedFloat(Minimum=4, Maximum=5))
   )
   /SetVariables(Particles.SpriteSize=&Multiply_Vector2(
     A=$Particles.SpriteSize
     B=&Vector2DFromFloat(Value=HlslExpression() {hlslExpression="Particles.MaterialRandom > .98 ? 2.5 : 1"})
    )) {isSetParameters=true}
   /Niagara/Modules/Spawn/Location/SphereLocation(MaskSpawnGroup=false, Module.Override Randomness=false, Sphere Radius=20, Surface Only=false)
   /SetVariables(Particles.Random:NiagaraBool=/Niagara/DynamicInputs/Random/RandomBool(Randomness Mode:ENiagaraRandomnessMode=::NewEnumerator1, Random Seed:NiagaraInt32=0, Fixed Random Seed:NiagaraBool=false)) {isSetParameters=true}
  ]
  particleUpdateScript [
   &ParticleState()
   /Niagara/Modules/Update/Forces/CurlNoiseForce(Noise Strength=1.5, Pan Noise=false)
   &SkeletalMeshLocation(
    Bone / Socket Index=$Particles.SpawnGroup
    Bone Sampling Mode=::NewEnumerator4
    Inherit Velocity=false
    Interpolate Mesh Sampling=false
    MaskSpawnGroup=false
    Sampled Position Offset=&Add_Vector(
     A=&Add_Vector(A=Linked() {linkedVariable {name=Particles.Initial.Position, type=NiagaraPosition}}, B=$Transient.PhysicsForce)
     B=&TransformVector(
      Vector=&MakeVector(
       X=0
       Y=0
       Z=&Lerp_Float(A=0, B=40, Alpha=&Multiply_Float(A=$Particles.NormalizedAge, B=$Particles.NormalizedAge))
      )
      Source Space=::World
      Destination Space=::World
     )
    )
    Sampled Position Offset Coordinate Space:ENiagaraMeshTransforms=::NewEnumerator2
    Skeletal Mesh=$User.SkeletalMesh
   ) {module=SkeletalMeshLocation002}
   &CameraOffset(Camera Offset Amount=-5)
   &ScaleSpriteSize(Scale Factor=&MakeCustomVector2DFromBool(
     Bool=$Particles.Random
     True Vector2D=&Vector2DFromFloat(Value=&FloatFromCurve(FloatCurve=DataInterface() {propertyValues {Curve {keys [
           <at 0 0 RCIM_Cubic>
           {arriveTangent=9.6335, interpMode=RCIM_Cubic, leaveTangent=2.7827, tangentMode=RCTM_Break, time=0.15, value=1}
           {arriveTangent=1.2288, interpMode=RCIM_Cubic, leaveTangent=1.2288, time=0.3322, value=0.4478}
           {arriveTangent=4.272, interpMode=RCIM_Cubic, leaveTangent=0.772, tangentMode=RCTM_Break, time=0.5569, value=1.5}
           {arriveTangent=-1.7123, interpMode=RCIM_Cubic, leaveTangent=-1.7123, time=0.6943, value=0.7273}
           {arriveTangent=2.7202, interpMode=RCIM_Cubic, leaveTangent=-2.3793, tangentMode=RCTM_Break, time=0.8489, value=1}
           <at 1 0 RCIM_Cubic>
          ]}}}))
     False Vector2D=&Vector2DFromFloat(Value=&FloatFromCurve(FloatCurve=DataInterface() {propertyValues {Curve {keys [
           <at 0 0 RCIM_Cubic>
           {arriveTangent=9.6335, interpMode=RCIM_Cubic, leaveTangent=2.7827, tangentMode=RCTM_Break, time=0.15, value=0.9794}
           {arriveTangent=0.0322, interpMode=RCIM_Cubic, leaveTangent=0.0322, time=0.5397, value=0.5}
           {arriveTangent=2.7202, interpMode=RCIM_Cubic, leaveTangent=-2.3793, tangentMode=RCTM_Break, time=0.7888, value=1}
           <at 1 0 RCIM_Cubic>
          ]}}}))
    ))
  ]
  rendererClasses [&NiagaraSpriteRendererProperties]
  renderers [{rendererClass=&NiagaraSpriteRendererProperties}]
 }
 Hand_Smoke {
  emitterSpawnScript [&FN_PreserveSkydiveUserVariables() {module=PreserveUserVariables}]
  emitterUpdateScript [
   &EmitterState(
    Inactive Response=::NewEnumerator1
    Life Cycle Mode=::NewEnumerator1
    Loop Behavior=::NewEnumerator1
    Loop Duration Mode=::NewEnumerator1
    UseLoopDelay=false
   )
   &SpawnBurst_Instantaneous(Spawn Count=2, Use Spawn Probability=false)
  ]
  particleSpawnScript [
   &FN_SetSpawnGroup_v2(Number of Spawn Groups=2)
   &InitializeParticle(
    Color=<rgba 4.5963 6 0 1>
    Lifetime=1.5
    Mesh Scale=<5 5 5>
    Module.Write Color=true
    Module.Write Lifetime=true
    Module.Write Mass=true
    Module.Write Material Random=true
    Module.Write Position=false
    Module.Write Scale=false
    Module.Write SpriteRotation=true
    Module.Write SpriteSize=true
    Position=&SimulationPosition()
    Sprite Size=<45 45>
   )
  ]
  particleUpdateScript [
   &ParticleState(Kill Particles When Lifetime Has Elapsed=false)
   &SkeletalMeshLocation(
    Bone / Socket Index=$Particles.SpawnGroup
    Bone Sampling Mode=::NewEnumerator4
    Inherit Velocity=false
    Interpolate Mesh Sampling=false
    MaskSpawnGroup=false
    Sampled Position Offset=&Add_Vector(
     A=Linked() {linkedVariable {name=Particles.Initial.Position, type=NiagaraPosition}}
     B=&MaskVectorBySpawnGroup(Masked Vector=<8 0 0>, Passthrough Vector=<-8 0 0>)
    )
    Sampled Position Offset Coordinate Space:ENiagaraMeshTransforms=::NewEnumerator3
    Skeletal Mesh=$User.SkeletalMesh
   ) {module=SkeletalMeshLocation001}
   &ScaleSpriteSize(Scale Factor=&Vector2DFromFloat(Value=&FloatFromCurve(FloatCurve=DataInterface() {propertyValues {Curve {keys [<at 0 0>,<at 0.15 1>,<at 0.75 1>]}}})))
   &CameraOffset(Camera Offset Amount=10)
   &FN_Light_v3(
    Color=<rgba 7.7626 10 0 1>
    Create Color Attribute=true
    Create Exponent Attribute=true
    Create Radius Attribute=true
    Enable Particle Light Enable=true
    Radius=75
   )
  ]
  rendererClasses [&NiagaraSpriteRendererProperties, &NiagaraLightRendererProperties]
  renderers [
   {rendererClass=&NiagaraSpriteRendererProperties}
   {rendererClass=&NiagaraLightRendererProperties, rendererIndex=1}
  ]
 }
 Hand_Smoke_Black {
  emitterSpawnScript [&FN_PreserveSkydiveUserVariables() {module=PreserveUserVariables}]
  emitterUpdateScript [
   &EmitterState()
   &SpawnRate(SpawnRate=20, Use Spawn Probability=false)
  ]
  particleSpawnScript [
   &FN_SetSpawnGroup_v2(Number of Spawn Groups=2)
   &InitializeParticle(
    Color=<rgba 0 0 0 0.3>
    Lifetime=1.5
    Mesh Scale=<5 5 5>
    Module.Write Color=true
    Module.Write Lifetime=true
    Module.Write Mass=true
    Module.Write Material Random=true
    Module.Write Position=false
    Module.Write Scale=false
    Module.Write SpriteRotation=true
    Module.Write SpriteSize=true
    Position=&SimulationPosition()
    Sprite Rotation=&UniformRangedFloat(Minimum=0, Maximum=360)
    Sprite Size=&Vector2DFromFloat(Value=&UniformRangedFloat(Minimum=40, Maximum=50))
   )
   &SphereLocation(MaskSpawnGroup=false, Module.Override Randomness=false, Sphere Radius=5, Surface Only=false)
  ]
  particleUpdateScript [
   &ParticleState()
   &SkeletalMeshLocation(
    Bone / Socket Index=$Particles.SpawnGroup
    Bone Sampling Mode=::NewEnumerator4
    Inherit Velocity=false
    Interpolate Mesh Sampling=false
    MaskSpawnGroup=false
    Sampled Position Offset=&Add_Vector(
     A=Linked() {linkedVariable {name=Particles.Initial.Position, type=NiagaraPosition}}
     B=&Add_Vector(
      A=&VectorFromCurve(VectorCurve=DataInterface() {propertyValues {XCurve {keys [<at 0 0>,<at 1 0>]}, YCurve {keys [<at 0 0>,<at 1 0>]}, ZCurve {keys [<at 0 0>,<at 1 10>]}}})
      B=<0 0 0>
     )
    )
    Skeletal Mesh=$User.SkeletalMesh
   ) {module=SkeletalMeshLocation001}
   &DynamicMaterialParameters(
    Index 0 Param 1=&FloatFromCurve(FloatCurve=DataInterface() {propertyValues {Curve {keys [<at 0 0>,<at 1 1>]}}})
    Module.Param0WriteEnabled=true
    Module.Param1WriteEnabled=true
    Module.Param2WriteEnabled=true
    Module.Param3WriteEnabled=true
   )
   &ScaleSpriteSize(Scale Factor=&Vector2DFromFloat(Value=&FloatFromCurve(FloatCurve=DataInterface() {propertyValues {Curve {keys [<at 0 0>,<at 0.15 1>,<at 0.75 1>,<at 1 0>]}}})))
   /Niagara/Modules/Update/SubUV/SubUVAnimation(Number Of Frames=24)
  ]
  rendererClasses [&NiagaraSpriteRendererProperties]
  renderers [{rendererClass=&NiagaraSpriteRendererProperties}]
 }
 Lightning {
  emitterUpdateScript [
   &EmitterState()
   &SpawnRate(SpawnRate=&UniformRangedFloat(Minimum=0, Maximum=20), Use Spawn Probability=false)
  ]
  particleSpawnScript [
   &FN_SetSpawnGroup_v2(Number of Spawn Groups=2)
   &InitializeParticle(
    Color=&MakeLinearColorFromVectorAndFloat(
     "Vector (RGB)"=&MakeVectorFromLinearColorRGB(Color=<rgba 6.21 8 0 1>)
     "Float (Alpha)"=&UniformRangedFloat(Minimum=0.25, Maximum=1)
    )
    Lifetime=&UniformRangedFloat(Minimum=0.45, Maximum=0.65)
    Module.Write Color=true
    Module.Write Lifetime=true
    Module.Write Mass=false
    Module.Write Material Random=true
    Module.Write Position=false
    Module.Write Scale=false
    Module.Write SpriteRotation=true
    Module.Write SpriteSize=true
    Sprite Rotation=&NormalizedAngleToDegrees(NormalizedAngle=&UniformRangedFloat(Minimum=0, Maximum=1))
    Sprite Size=&Vector2DFromFloat(Value=&UniformRangedFloat(Minimum=20, Maximum=32))
   )
   /SetVariables(Particles.SubImageIndex=&UniformRangedFloat(Minimum=0, Maximum=3)) {isSetParameters=true}
   &DynamicMaterialParameters(
    Index 0 Param 1=&UniformRangedFloat(Minimum=0, Maximum=0)
    Index 0 Param 2=&UniformRangedFloat(Minimum=1, Maximum=2)
    Module.Param0WriteEnabled=true
    Module.Param1WriteEnabled=true
    Module.Param2WriteEnabled=false
    Module.Param3WriteEnabled=false
   )
   &CameraOffset(Camera Offset Amount=10)
   &SphereLocation(MaskSpawnGroup=false, Module.Override Randomness=false, Sphere Radius=15, Surface Only=false)
  ]
  particleUpdateScript [
   &ParticleState(Kill Particles When Lifetime Has Elapsed=false, Let Infinitely Lived Particles Die When Emitter Deactivates=true)
   &SkeletalMeshLocation(
    Bone / Socket Index=$Particles.SpawnGroup
    Bone Sampling Mode=::NewEnumerator4
    Inherit Velocity=false
    Interpolate Mesh Sampling=false
    MaskSpawnGroup=false
    Sampled Position Offset=&Add_Vector(
     A=Linked() {linkedVariable {name=Particles.Initial.Position, type=NiagaraPosition}}
     B=&MaskVectorBySpawnGroup(Masked Vector=<8 0 0>, Passthrough Vector=<-8 0 0>)
    )
    Sampled Position Offset Coordinate Space:ENiagaraMeshTransforms=::NewEnumerator3
    Skeletal Mesh=$User.SkeletalMesh
   ) {module=SkeletalMeshLocation001}
   &DynamicMaterialParameters(
    Index 0 Param 3=&FloatFromCurve(FloatCurve=DataInterface() {propertyValues {Curve {keys [<at 0 1 RCIM_Cubic>,<at 0.2 0 RCIM_Cubic>,<at 1 1>]}}})
    Module.Param0WriteEnabled=false
    Module.Param1WriteEnabled=false
    Module.Param2WriteEnabled=true
    Module.Param3WriteEnabled=false
   ) {module=DynamicMaterialParameters001}
   &ScaleColor(
    Scale Alpha=&FloatFromCurve(FloatCurve=DataInterface() {propertyValues {Curve {keys [<at 0.75 1 RCIM_Cubic>,<at 1 0>]}}})
    Scale RGB=&MakeVectorFromLinearColorRGB(Color=&ColorFromCurve(ColorCurve=DataInterface() {propertyValues {
        AlphaCurve {keys []}
        BlueCurve {keys [<at 0 3 RCIM_Cubic>,<at 1 1 RCIM_Cubic>]}
        GreenCurve {keys [<at 0 2.7135 RCIM_Cubic>,<at 1 0.8177 RCIM_Cubic>]}
        RedCurve {keys [<at 0 2.5625 RCIM_Cubic>,<at 1 0.8188 RCIM_Cubic>]}
       }}))
    ScaleA=true
    ScaleRGB=true
   )
  ]
  rendererClasses [&NiagaraSpriteRendererProperties]
  renderers [{rendererClass=&NiagaraSpriteRendererProperties}]
 }
 SmokeRibbon003 {
  emitterUpdateScript [
   &EmitterState()
   &FN_SpawnRibbon_v2(Number of Ribbons=2, Spawn Mode=::NewEnumerator0, SpawnRate=50)
   /Game/Effects/Niagara/Modules/Skydiving/FN_DivingAttributes/FN_EmitterDivingAttributes_v1(Accumulate Dive=-0.0025, Accumulate Rest=0, SkeletalMesh:NiagaraDataInterfaceSkeletalMesh=$User.SkeletalMesh)
  ]
  particleSpawnScript [
   &FN_InitializeRibbon_v3(
    Color:LinearColor=&MultiplyLinearColor(Linear Color=<rgba 0.7991 1 0.1413 1>, Scale Factor=<1 1 1 1>)
    Custom Offset=true
    Enable Skeletal Position=true
    Offset Amount=1
    Ribbon Width=34
   )
   &FN_SetSpawnGroup_v2(Number of Spawn Groups=2)
   /SetVariables(Particles.RibbonNumber:NiagaraInt32=$Particles.SpawnGroup) {isSetParameters=true}
   /SetVariables(Particles.Color=&LerpLinearColors(
     StartColor=&ColorFromCurve(
      ColorCurve=DataInterface() {propertyValues {
        AlphaCurve {keys [<at 0 1 RCIM_Cubic>]}
        BlueCurve {keys [<at 0 0.1413 RCIM_Cubic>,<at 0.1 0>,<at 0.2 0.1413 RCIM_Cubic>]}
        GreenCurve {keys [<at 0 1 RCIM_Cubic>,<at 0.1 8>,<at 0.2 1 RCIM_Cubic>]}
        RedCurve {keys [<at 0 0.7991 RCIM_Cubic>,<at 0.1 6.1284>,<at 0.2 0.7991 RCIM_Cubic>]}
       }}
      CurveIndex=$Emitter.LoopedAge
     )
     EndColor=&ColorFromCurve(
      ColorCurve=DataInterface() {propertyValues {
        AlphaCurve {keys [<at 0 1 RCIM_Cubic>]}
        BlueCurve {keys [<at 0 0.1413 RCIM_Cubic>,<at 0.1 0>,<at 0.2 0.1413 RCIM_Cubic>]}
        GreenCurve {keys [<at 0 1 RCIM_Cubic>,<at 0.1 8>,<at 0.2 1 RCIM_Cubic>]}
        RedCurve {keys [<at 0 0.7991 RCIM_Cubic>,<at 0.1 6.1284>,<at 0.2 0.7991 RCIM_Cubic>]}
       }}
      CurveIndex=$System.LoopedAge
     )
     LerpFactor=&MakeFloatFromInt(Int=$Particles.SpawnGroup)
    )) {isSetParameters=true}
   &AddVelocity(Velocity=&Lerp_Vector(A=<0 0 400>, B=<0 0 400>, Alpha=$Emitter.Dive))
   /SetVariables(Particles.SmokeColor:LinearColor=<rgba 0 0 0 0.6>) {isSetParameters=true}
  ]
  particleUpdateScript [
   &ScaleVelocity(Velocity Scale=&MakeVector(
     X=&FloatFromCurve(FloatCurve=DataInterface() {propertyValues {Curve {keys [<at 0 1 RCIM_Cubic>,<at 0.3 0 RCIM_Cubic>]}}})
     Y=&FloatFromCurve(FloatCurve=DataInterface() {propertyValues {Curve {keys [<at 0 1 RCIM_Cubic>,<at 0.3 0 RCIM_Cubic>]}}})
     Z=1
    ))
   &DynamicMaterialParameters(
    Index 0 Param 1=&Add_Float(A=$Particles.Initial.MaterialRandom, B=$Emitter.IntegratedDive)
    Index 0 Param 3=&Lerp_Float(A=1, B=0.2, Alpha=$Emitter.Dive)
    Module.Param0WriteEnabled=true
    Module.Param1WriteEnabled=true
    Module.Param2WriteEnabled=true
    Module.Param3WriteEnabled=true
   ) {module=DynamicMaterialParameters001}
   &SolveForcesAndVelocity(Clamp Velocity=false, Limit Acceleration=false)
   &FN_UpdateRibbon(
    Enable Skeletal Position=true
    Initial Color:LinearColor=$Particles.Initial.Color
    Initial RibbonWidth:NiagaraFloat=$Particles.Initial.RibbonWidth
   )
   &ScaleColor(
    Scale Alpha=&FloatFromCurve(FloatCurve=DataInterface() {propertyValues {Curve {keys [<at 0 1>,<at 0.8 1>]}}})
    Scale RGB=&VectorFromCurve(VectorCurve=DataInterface() {propertyValues {
       XCurve {keys [<at 0.0618 1>,<at 0.1318 1>,<at 0.25 1>,<at 0.75 0>]}
       YCurve {keys [<at 0.25 1>,<at 0.75 1>]}
       ZCurve {keys [<at 0.0618 1>,<at 0.1318 1>,<at 0.25 1>,<at 0.75 0>]}
      }})
    ScaleA=true
    ScaleRGB=true
   )
   &ScaleColor(
    Scale RGB=&VectorFromFloat(Value=&FloatFromCurve(FloatCurve=DataInterface() {propertyValues {Curve {keys [
          {arriveTangent=0.1804, interpMode=RCIM_Cubic, leaveTangent=0.1804, tangentMode=RCTM_User, time=0, value=6}
          {arriveTangent=-0.4145, interpMode=RCIM_Cubic, leaveTangent=-0.4145, tangentMode=RCTM_User, time=0.3, value=1}
         ]}}}))
    ScaleA=true
    ScaleRGB=true
   ) {module=ScaleColor001}
   &ScaleRibbonWidth(Ribbon Width Scale=&FloatFromCurve(FloatCurve=DataInterface() {propertyValues {Curve {keys [<at 0.011 1>, {arriveTangent=-1.0111, interpMode=RCIM_Cubic, leaveTangent=-1.0111, time=0.5, value=1}, <at 1 0 RCIM_Cubic>]}}}))
  ]
  rendererClasses [&NiagaraRibbonRendererProperties]
  renderers [
   {rendererClass=&NiagaraRibbonRendererProperties}
   {rendererClass=&NiagaraRibbonRendererProperties, rendererIndex=1}
  ]
 }
 UpwardDots003 {
  emitterSpawnScript [&FN_PreserveSkydiveUserVariables() {module=PreserveUserVariables}]
  emitterUpdateScript [
   &EmitterState(
    Inactive Response=::NewEnumerator1
    Life Cycle Mode=::NewEnumerator1
    Loop Delay=&UniformRangedFloat(Minimum=0, Maximum=0.45)
    Loop Duration=&UniformRangedFloat(Minimum=0.1, Maximum=0.35)
    Recalculate Duration Each Loop=true
    UseLoopDelay=true
   )
   &SpawnBurst_Instantaneous(
    Spawn Count=&UniformRangedInt(Minimum=10, Maximum=20)
    Spawn Group:NiagaraInt32=&UniformRangedInt(Minimum=0, Maximum=1, Randomness Mode=::NewEnumerator1, Random Seed=$Emitter.LoopCount)
    Use Spawn Probability=false
   ) {module=SpawnBurst_Instantaneous002}
   &FN_EmitterDivingAttributes_v1(Accumulate Dive=-0.0025, Accumulate Rest=0, SkeletalMesh:NiagaraDataInterfaceSkeletalMesh=$User.SkeletalMesh)
  ]
  particleSpawnScript [
   &FN_SetSpawnGroup_v2(Number of Spawn Groups=2)
   &SkeletalMeshLocation(
    Bone / Socket Index=&SpawnGroup()
    Bone Sampling Mode=::NewEnumerator4
    Inherit Velocity=false
    MaskSpawnGroup=false
    Sampled Velocity Coordinate Space:ENiagaraCoordinateSpace=::World
    Skeletal Mesh=$User.SkeletalMesh
   )
   /Game/Effects/Niagara/Modules/Orbit/FN_Orbit_Init_Variables_v2()
   &InitializeParticle(
    Color=<rgba 0 0 0 1>
    Lifetime=&UniformRangedFloat(Minimum=0.5, Maximum=0.75)
    Mesh Scale=<5 5 5>
    Module.Write Color=true
    Module.Write Lifetime=true
    Module.Write Mass=true
    Module.Write Material Random=true
    Module.Write Position=false
    Module.Write Scale=false
    Module.Write SpriteRotation=false
    Module.Write SpriteSize=true
    Position=&SimulationPosition()
    Sprite Size=&Vector2DFromFloat(Value=&UniformRangedFloat(Minimum=4, Maximum=5))
   )
   &AddVelocity(Velocity=&Lerp_Vector(A=<0 0 200>, B=<0 0 400>, Alpha=$Emitter.Dive))
   &ApplyInitialForces()
   /SetVariables(Particles.Random:NiagaraBool=&RandomBool(Randomness Mode:ENiagaraRandomnessMode=::NewEnumerator1, Random Seed:NiagaraInt32=0, Fixed Random Seed:NiagaraBool=false)) {isSetParameters=true}
   /SetVariables(Particles.RotRate:Vector2f=<-0.25 0.25>) {isSetParameters=true}
  ]
  particleUpdateScript [
   &ParticleState()
   /Game/Effects/Niagara/Modules/Orbit/FN_Orbit_v4(
    Height min/max:Vector2f=<0 40>
    Radius min/max:Vector2f=&Multiply_Vector2(
     A=<15 15>
     B=&Vector2DFromFloat(Value=&OneMinusFloat(Float=$Particles.NormalizedAge))
    )
    Rotation Rate min/max:Vector2f=$Particles.RotRate
    To Axis:Vector3f=$Particles.Initial.Velocity
   )
   &ScaleVelocity(Velocity Scale=&MakeVector(
     X=&FloatFromCurve(FloatCurve=DataInterface() {propertyValues {Curve {keys [<at 0 1 RCIM_Cubic>,<at 0.3 0 RCIM_Cubic>]}}})
     Y=&FloatFromCurve(FloatCurve=DataInterface() {propertyValues {Curve {keys [<at 0 1 RCIM_Cubic>,<at 0.3 0 RCIM_Cubic>]}}})
     Z=1
    ))
   &SolveForcesAndVelocity(Clamp Velocity=false, Limit Acceleration=false)
   &CameraOffset(Camera Offset Amount=-5)
   &ScaleSpriteSize(Scale Factor=&MakeCustomVector2DFromBool(
     Bool=$Particles.Random
     True Vector2D=&Vector2DFromFloat(Value=&FloatFromCurve(FloatCurve=DataInterface() {propertyValues {Curve {keys [
           <at 0 0.5 RCIM_Cubic>
           {arriveTangent=9.6335, interpMode=RCIM_Cubic, leaveTangent=2.7827, tangentMode=RCTM_Break, time=0.15, value=1}
           {arriveTangent=1.2288, interpMode=RCIM_Cubic, leaveTangent=1.2288, time=0.3322, value=0.4477}
           {arriveTangent=4.272, interpMode=RCIM_Cubic, leaveTangent=0.772, tangentMode=RCTM_Break, time=0.5569, value=1.5}
           {arriveTangent=-1.7123, interpMode=RCIM_Cubic, leaveTangent=-1.7123, time=0.6943, value=0.7273}
           {arriveTangent=2.7202, interpMode=RCIM_Cubic, leaveTangent=-2.3793, tangentMode=RCTM_Break, time=0.8489, value=1}
           <at 1 0.5 RCIM_Cubic>
          ]}}}))
     False Vector2D=&Vector2DFromFloat(Value=&FloatFromCurve(
       FloatCurve=DataInterface() {propertyValues {Curve {keys [
           {arriveTangent=1.1559, interpMode=RCIM_Cubic, leaveTangent=1.1559, tangentMode=RCTM_User, time=0, value=0.5}
           {arriveTangent=9.6335, interpMode=RCIM_Cubic, leaveTangent=2.7827, tangentMode=RCTM_Break, time=0.15, value=0.9794}
           {arriveTangent=0.0322, interpMode=RCIM_Cubic, leaveTangent=0.0322, time=0.5397, value=0.5}
           {arriveTangent=2.7202, interpMode=RCIM_Cubic, leaveTangent=-2.3793, tangentMode=RCTM_Break, time=0.7888, value=1}
           <at 1 0.5 RCIM_Cubic>
          ]}}}
       Scale Curve=&FloatFromCurve(FloatCurve=DataInterface() {propertyValues {Curve {keys [
            {arriveTangent=5.8968, interpMode=RCIM_Cubic, leaveTangent=5.8968, tangentMode=RCTM_User, time=0, value=0}
            <at 0.25 1.1>
            {arriveTangent=-1.4667, interpMode=RCIM_Cubic, leaveTangent=-1.4667, time=0.75, value=1}
            {arriveTangent=-6.2476, interpMode=RCIM_Cubic, leaveTangent=-6.2476, tangentMode=RCTM_User, time=1, value=0}
           ]}}})
      ))
    ))
  ]
  rendererClasses [&NiagaraSpriteRendererProperties]
  renderers [{rendererClass=&NiagaraSpriteRendererProperties}]
 }
}
summary {
 systemName=NS_Contrail_Hightower_Date
 emitters [
  {emitterName=Hand_Dots, rendererClasses [&NiagaraSpriteRendererProperties]}
  {emitterName=Hand_Smoke_Black, rendererClasses [&NiagaraSpriteRendererProperties]}
  {emitterName=Hand_Smoke, rendererClasses [&NiagaraSpriteRendererProperties, &NiagaraLightRendererProperties]}
  {emitterName=SmokeRibbon003, rendererClasses [&NiagaraRibbonRendererProperties]}
  {emitterName=Lightning, rendererClasses [&NiagaraSpriteRendererProperties]}
  {emitterName=UpwardDots003, rendererClasses [&NiagaraSpriteRendererProperties]}
 ]
 userVariables [
  User.bIsFrontEndPreview {type=NiagaraBool}
  User.PlayerSpeed {type=NiagaraFloat, defaultValue=0, description=""}
  User.RibbonIDOffset {type=NiagaraInt32}
  User.SkeletalMesh {
   type=NiagaraDataInterfaceSkeletalMesh
   defaultValue {
    struct=NiagaraExt_VariableValue_DataInterface
    value {dataInterface=&self:NiagaraDataInterfaceSkeletalMesh_0}
   }
  }
 ]
}
